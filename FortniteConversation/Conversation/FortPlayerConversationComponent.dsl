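; Player conversation component Blueprint that responds when the Fortnite conversation screen widget is initialized.
; It checks that the passed widget is valid, stores it as the screen widget and fires the screen-created delegate with it.
; An invalid widget instead writes a shipping log message saying it was not a CommonActivatableWidget, though the check itself only tests validity.
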(event Audio|EventFortConversationScreenInitialized
  (params (WidgetToUpdate :type Widget%20Object%20Reference))
  (Utilities|IsValid WidgetToUpdate
    (:Is%20Valid
      (Variables|Default|SetFortScreenWidget WidgetToUpdate)
      (Default|CallFortScreenCreatedDelegate WidgetToUpdate))
    (:Is%20Not%20Valid
      (Fortnite|FortShippingLog self "Event FortConversationScreenInitialized: The widget passed was not a CommonActivatableWidget." true))))
