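; Conversation task node for a Fortnite NPC dialogue line, showing a speaker's message with voice-over.
; ExecuteTaskNode resolves the contextual message, formats in styled speaker and listener names ("???" if missing), then pauses the conversation and sends client choices.
; Client effects trigger contextual or tag-based VO on the NPC through BPI_NonParticipant.
; Helpers tint the node by speaker debug color and swap the speaker and listener tags.
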(fn ExecuteClientEffects
  (params (Context :type Conversation%20Context%20Structure%20%28by%20ref%29))
  (CallFunction|HandleVOPlayback Context)
  (return))

(fn ExecuteTaskNode
  (params (Context :type Conversation%20Context%20Structure%20%28by%20ref%29))
  (bind (_message _metadataparameters) (Utilities|Struct|BreakContextualMessageQueryResult (CallFunction|ResolveMessageforContext self (Variables|Default|GetSpeakerTag) Context)))
  (Utilities|FlowControl|Sequence
    (:then_0
      (bind _returnvalue (CallFunction|GetParticipantNameStyled (Variables|GetContext) (Variables|Default|GetSpeakerTag)))
      (LocalVariables|SetSpeakerDisplayName _returnvalue))
    (:then_1
      (bind _returnvalue_2 (CallFunction|GetParticipantNameStyled (Variables|GetContext) (Variables|Default|GetListenerTag)))
      (LocalVariables|SetListenerDisplayName _returnvalue_2))
    (:then_2
      (bind _returnvalue_3 (Conversation|PauseConversationandSendClientChoices (Variables|GetContext) (Utilities|Struct|MakeClientConversationMessage (Variables|Default|GetSpeakerTag) (LocalVariables|GetSpeakerDisplayName) (Utilities|Text|FormatText _message (LocalVariables|GetSpeakerDisplayName) (LocalVariables|GetListenerDisplayName)) _metadataparameters)))
      (return _returnvalue_3))))

(fn GetNodeBodyColor
  (return (Conversation|GetDebugParticipantColor (Variables|Default|GetSpeakerTag)) true))

(fn GetParticipantNameStyled
  (params (Context :type Conversation%20Context%20Structure%20%28by%20ref%29) (ParticipantTag :type Gameplay%20Tag%20Structure))
  (bind _returnvalue (Conversation|GetConversationParticipant Context ParticipantTag))
  (Utilities|IsValid _returnvalue
    (:Is%20Valid
      (bind _returnvalue_2 (Conversation|GetParticipantDisplayName _returnvalue))
      (LocalVariables|SetRawName _returnvalue_2))
    (:Is%20Not%20Valid
      (LocalVariables|SetRawName "NSLOCTEXT(\"[68ECC0C546A9E9BA3AA37C9B63BEEEA1]\", \"DA9C661C494FB633E97655B796ECF5F8\", \"???\")")))
  (return (Utilities|Text|FormatText "NSLOCTEXT(\"[68ECC0C546A9E9BA3AA37C9B63BEEEA1]\", \"907BE04440B40658EAE85E808E061E57\", \"<Name>{0}</>\")" (LocalVariables|GetRawName))))

(fn HandleVOPlayback
  (params (Context :type Conversation%20Context%20Structure%20%28by%20ref%29))
  (bind (_message _metadataparameters _metadatatags) (Utilities|Struct|BreakContextualMessageQueryResult (CallFunction|ResolveMessageforContext (Variables|Default|GetSpeakerTag) Context)))
  (bind _returnvalue (Class|FortConversationContextLibrary|GetParticipantActorNPC Context))
  (if (Utilities|DoesObjectImplementInterface _returnvalue "/Game/Athena/AI/NPCs/Base/BPI_NonParticipant.BPI_NonParticipant_C")
    ;[0.9852 1 0.4314] Contextual VO
    (BPI_NonParticipant|AttemptPlayVocalReinforcementSound(Message) _returnvalue _metadatatags)
    (if (GameplayTags|IsGameplayTagValid (Variables|Default|GetVoiceOverTag))
      (BPI_NonParticipant|PlayVoiceSoundbyTag(Message) _returnvalue (Variables|Default|GetVoiceOverTag)))))

(fn SwapSpeakerAndListener
  (LocalVariables|SetTemp (Variables|Default|GetSpeakerTag))
  (Variables|SetBy-RefVar (Variables|Default|GetSpeakerTag) (Variables|Default|GetListenerTag))
  (Variables|SetBy-RefVar (Variables|Default|GetListenerTag) (LocalVariables|GetTemp)))
